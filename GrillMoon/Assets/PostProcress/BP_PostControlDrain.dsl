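; Actor Blueprint that drives the Drain post-process effect through the MPC_PostProcress_Drain parameter collection.
; Each update randomizes a value within the collection's PreFrameRandomBP range, accumulates texture time scaled by TextureSpeedBP, and writes both back with player pawn 0's world position.
; This logic is duplicated: custom event UpdateMPC runs it when bDebug is off, Tick when on.
; Tick prints "DrainPostDebugOn" every frame, regardless of bDebug.
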(event EventBeginPlay)

(event Collision|EventActorBeginOverlap
  (params (OtherActor :type Actor%20Object%20Reference)))

(event Custom|UpdateMPC
  (bind _returnvalue (Game|GetPlayerPawn 0))
  (bind _returnvalue_location (Math|Transform|BreakTransform (Transformation|GetActorTransform _returnvalue)))
  (if (not (Variables|Default|GetbDebug))
    (bind _returnvalue_2 (Rendering|Material|GetScalarParameterValue "/GrillMoon/Assets/PostProcress/MPC_PostProcress_Drain.MPC_PostProcress_Drain" "PreFrameRandomBP" :!class /Script/Engine.KismetMaterialLibrary))
    (Rendering|Material|SetScalarParameterValue "/GrillMoon/Assets/PostProcress/MPC_PostProcress_Drain.MPC_PostProcress_Drain" "PreFrameRandomMAT" (Math|Random|RandomFloatinRange 0.0 _returnvalue_2) :!class /Script/Engine.KismetMaterialLibrary)
    (bind _returnvalue_3 (Rendering|Material|GetScalarParameterValue "/GrillMoon/Assets/PostProcress/MPC_PostProcress_Drain.MPC_PostProcress_Drain" "TextureSpeedBP" :!class /Script/Engine.KismetMaterialLibrary))
    (Variables|Default|SetTextureTime (+ (Variables|Default|GetTextureTime) (* _returnvalue_3 (Utilities|Time|GetWorldDeltaSeconds))))
    (Rendering|Material|SetScalarParameterValue "/GrillMoon/Assets/PostProcress/MPC_PostProcress_Drain.MPC_PostProcress_Drain" "TextureTimeMAT" (Variables|Default|GetTextureTime) :!class /Script/Engine.KismetMaterialLibrary)
    (Utilities|IsValid _returnvalue
      (:Is%20Valid
        (Rendering|Material|SetVectorParameterValue "/GrillMoon/Assets/PostProcress/MPC_PostProcress_Drain.MPC_PostProcress_Drain" "Player_WP" (Math|Conversions|ToLinearColor(Vector) _returnvalue_location) :!class /Script/Engine.KismetMaterialLibrary)))))

(event EventTick
  (params (DeltaSeconds :type Float%20%28single-precision%29))
  (bind _returnvalue (Game|GetPlayerPawn 0))
  (bind _returnvalue_location (Math|Transform|BreakTransform (Transformation|GetActorTransform _returnvalue)))
  (Development|PrintString "DrainPostDebugOn" true true "(R=1.000000,G=0.612965,B=0.000000,A=1.000000)" 0.0)
  (if (Variables|Default|GetbDebug)
    (bind _returnvalue_2 (Rendering|Material|GetScalarParameterValue "/GrillMoon/Assets/PostProcress/MPC_PostProcress_Drain.MPC_PostProcress_Drain" "PreFrameRandomBP" :!class /Script/Engine.KismetMaterialLibrary))
    (Rendering|Material|SetScalarParameterValue "/GrillMoon/Assets/PostProcress/MPC_PostProcress_Drain.MPC_PostProcress_Drain" "PreFrameRandomMAT" (Math|Random|RandomFloatinRange 0.0 _returnvalue_2) :!class /Script/Engine.KismetMaterialLibrary)
    (bind _returnvalue_3 (Rendering|Material|GetScalarParameterValue "/GrillMoon/Assets/PostProcress/MPC_PostProcress_Drain.MPC_PostProcress_Drain" "TextureSpeedBP" :!class /Script/Engine.KismetMaterialLibrary))
    (Variables|Default|SetTextureTime (+ (Variables|Default|GetTextureTime) (* _returnvalue_3 (Utilities|Time|GetWorldDeltaSeconds))))
    (Rendering|Material|SetScalarParameterValue "/GrillMoon/Assets/PostProcress/MPC_PostProcress_Drain.MPC_PostProcress_Drain" "TextureTimeMAT" (Variables|Default|GetTextureTime) :!class /Script/Engine.KismetMaterialLibrary)
    (Utilities|IsValid _returnvalue
      (:Is%20Valid
        (Rendering|Material|SetVectorParameterValue "/GrillMoon/Assets/PostProcress/MPC_PostProcress_Drain.MPC_PostProcress_Drain" "Player_WP" (Math|Conversions|ToLinearColor(Vector) _returnvalue_location) :!class /Script/Engine.KismetMaterialLibrary)))))

(fn ConstructionScript)
