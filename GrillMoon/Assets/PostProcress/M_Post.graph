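# Post-process material applied before bloom that tints and boosts scene color and builds screen effects: a red slash line, a ghost-word texture and an inversion.
# MPC parameters drive tint, emissive, slash, word color and inversion, with bright pixels escaping the tint.
# A fixed world point is projected to screen UVs to position effects, and a world-space sphere mask is defined.
# Named reroutes link the stages; inversion uses a steep power curve.
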Material /GrillMoon/Assets/PostProcress/M_Post
graph {
 properties {
  bAutomaticallySetUsageInEditor=false
  blendableLocation=BL_SceneColorBeforeBloom
  materialDomain=MD_PostProcess
  usedShadingModels=Default Lit
 }
 nodes [
  n1 = NamedRerouteUsage()
  n2 = Power(Base=OneMinus(None=Saturate(None=@n1))) {constExponent=100}
  n3 = CollectionParameter() {collection=&/MPC_PostProcress, parameterName=Invert}
  Add(A=NamedRerouteUsage(), B=AppendVector(A=LinearInterpolate(A=@n1, B=@n2, Alpha=@n3), B=NamedRerouteUsage()))
  n4 = VectorParameter() {defaultValue=<rgba -36481.1719 49098.3867 15579.5156 50000>}
  Divide(A=@n4.RGB) {constB=100}
  n5 = Add(
   A=Multiply(A=NamedRerouteUsage(), B=NamedRerouteUsage())
   B=ComponentMask(None=CollectionParameter() {collection=&MPC_PostProcress, parameterName=SlashOffset}) {g=true, r=true}
  )
  n6 = OneMinus(None=Multiply(A=Abs(None=@n5)) {constB=2})
  n7 = CollectionParameter() {collection=&MPC_PostProcress, parameterName=SlashOffset}
  n8 = Multiply(A=ComponentMask(None=@n6) {g=true}, B=LinearInterpolate(A=ComponentMask(None=@n6) {r=true}, Alpha=ComponentMask(None=@n7) {b=true}))
  n9 = ComponentMask(None=@n7) {a=true}
  NamedRerouteDeclaration(None=Multiply(
    A=MaterialFunctionCall!(InRamp=@n8, CoreIntensity=@n9, DetailBalence, MaxOutPut) {materialFunction=/Hermes/Props/ClayTile/ClayTile_DragonGate_B/VFX/MF_MOON_InvertDistanceRamp}.Result
    B=VectorParameter() {parameterName=Param_1, defaultValue=<rgba 1 0 0.1825 1>}.RGB
   )) {name=Line, nodeColor=<rgba 0 1 0.625 1>}
  n10 = SceneTexture() {sceneTextureId=PPI_PostProcessInput0}
  NamedRerouteDeclaration(None=ComponentMask(None=@n10.Color) {a=true}) {name=InAlpha, nodeColor=<rgba 0 0.6016 1 1>}
  n11 = CollectionParameter() {collection=&MPC_PostProcress, parameterName=SceneBG_Tint}
  n12 = ComponentMask(None=@n10.Color) {b=true, g=true, r=true}
  n13 = Length(None=@n10.Color) {is=class MaterialExpressionLength}
  n14 = ScalarParameter() {defaultValue=30, parameterName=MaskSelectOffset}
  n15 = LinearInterpolate(A=Multiply(A=ComponentMask(None=@n11) {b=true, g=true, r=true}, B=@n12), B=@n12, Alpha=Saturate(None=Subtract(A=@n13, B=@n14)))
  n16 = Multiply(
   A=LinearInterpolate(A=Min(A=@n15), B=@n15, Alpha=NamedRerouteUsage())
   B=CollectionParameter() {collection=&MPC_PostProcress, parameterName=PostEmissiveMultiply}
  )
  n17 = NamedRerouteUsage()
  NamedRerouteDeclaration(None=Add(A=Add(A=@n16, B=@n17), B=NamedRerouteUsage())) {name=Stage1, nodeColor=<rgba 1 0 0.0937 1>}
  NamedRerouteDeclaration(None=Subtract(A=ScreenPosition().ViewportUV, B=NamedRerouteUsage())) {name=FixedViewUV, nodeColor=<rgba 0 1 0.2031 1>}
  n18 = Multiply(A=NamedRerouteUsage(), B=NamedRerouteUsage())
  n19 = TextureSample(UVs=MaterialFunctionCall(UVs=Add(A=@n18) {constB=0.5}, Texture Scale=ScalarParameter() {defaultValue=0.8, parameterName=TexSize}) {materialFunction=&ScaleUVsByCenter}.UVs) {samplerSource=SSM_Clamp_WorldGroupSettings, samplerType=SAMPLERTYPE_LinearColor, texture=&/T_GhostWord}
  n20 = CollectionParameter() {collection=&MPC_PostProcress, parameterName=WordColor}
  NamedRerouteDeclaration(None=Multiply(A=@n19.R, B=ComponentMask(None=@n20) {b=true, g=true, r=true})) {name=Word, nodeColor=<rgba 0.5938 0 1 1>}
  n21 = VectorParameter() {parameterName=WP, defaultValue=<rgba -21116 32736 23992 0>}
  n22 = CameraPositionWS()
  NamedRerouteDeclaration(None=MaterialFunctionCall!(
    Camera Relative Position=Subtract(A=@n21.RGB, B=@n22)
    X Camera to World Vector
    Y Camera to World Vector
    Z Camera to World Vector
    "Tan(FOV/2)*[1,Screen Res Y/X]"
   ) {materialFunction=/Engine/Functions/Engine_MaterialFunctions02/Math/ManualWorldToScreenUVsTransform}.Screen UVs) {name=EffectUVPos, nodeColor=<rgba 0.1484 0 1 1>}
  n23 = MaterialFunctionCall(Float2=ViewProperty() {property=MEVP_ViewSize}.Property) {materialFunction=&BreakOutFloat2Components}
  NamedRerouteDeclaration(None=AppendVector(A=Divide(A=@n23.R, B=@n23.R), B=Constant() {r=1})) {name=UV_ScaleFactor, nodeColor=<rgba 1 0 0.3281 1>}
  NamedRerouteDeclaration(None=SphereMask(A=WorldPosition().XYZ, B=@n4.RGB, Radius=@n4.RGB)) {name=WP_SphereMask, nodeColor=<rgba 0 0.7101 1 1>}
 ]
}
