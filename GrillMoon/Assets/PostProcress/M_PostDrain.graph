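# Post-process material for a "drain" screen effect centered on a gate's world position, applied before bloom and driven by MPC_PostProcress_Drain.
# It projects the gate to screen space and builds polar, panning UVs for tile textures that drive distortion and color overlays.
# Dithered, scaled scene-color resamples add blurred ghosting, with color-selective desaturation and a sphere-masked gate color.
# Randomized tint and vignette terms modulate color, alongside a ray-traced sphere player mask.
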Material /GrillMoon/Assets/PostProcress/M_PostDrain
graph {
 properties {
  bAutomaticallySetUsageInEditor=false
  bIsBlendable=false
  blendableLocation=BL_SceneColorBeforeBloom
  materialDomain=MD_PostProcess
  shadingModel=MSM_Unlit
  usedShadingModels=Unlit
 }
 nodes [
  CameraVectorWS()
  n1 = CollectionParameter() {collection=&/MPC_PostProcress_Drain, parameterName=Desaturation}
  ComponentMask(None=@n1) {b=true, g=true, r=true}
  n2 = NamedRerouteUsage()
  ConstantBiasScale(None=@n2)
  MaterialFunctionCall() {materialFunction=&DitherTemporalAA_Fortnite}
  n3 = CollectionParameter() {collection=&MPC_PostProcress_Drain, parameterName=MainTint}
  n4 = ComponentMask(None=@n3) {b=true, g=true, r=true}
  n5 = CollectionParameter() {collection=&MPC_PostProcress_Drain, parameterName=PreFrameRandomMAT}
  n6 = Multiply(A=@n4, B=@n5)
  n7 = ComponentMask(None=@n3) {a=true}
  n8 = NamedRerouteUsage()
  n9 = ComponentMask(None=Add(A=CollectionParameter() {collection=&MPC_PostProcress_Drain, parameterName=Vignette}, B=@n5)) {b=true, g=true, r=true}
  Multiply(
   A=Multiply(A=NamedRerouteUsage(), B=LinearInterpolate(A=@n4, B=@n6, Alpha=@n7))
   B=LinearInterpolate(B=Power(Base=@n8, Exp=@n9)) {constA=1, constAlpha=0.75}
  )
  n10 = Subtract(A=CameraPositionWS(), B=VectorParameter() {parameterName=Gate_WP, defaultValue=<rgba -57600 42904 9988 0>}.RGB)
  n11 = Divide(A=WorldPosition() {worldPositionShaderOffset=WPT_CameraRelative}.XYZ) {constB=100}
  NamedRerouteDeclaration(None=DotProduct(A=Normalize(VectorInput=@n10), B=Normalize(VectorInput=@n11))) {name=CamDotMask, nodeColor=<rgba 1 0.5859 0 1>}
  n12 = MaterialFunctionCall(Float2=Subtract(A=ScreenPosition().ViewportUV, B=NamedRerouteUsage())) {materialFunction=&BreakOutFloat2Components}
  n13 = Arctangent2Fast(Y=@n12.R, X=@n12.R)
  n14 = Multiply(A=MaterialFunctionCall() {materialFunction=&Pi}.Result) {constB=2}
  NamedRerouteDeclaration(None=AppendVector(A=NamedRerouteUsage(), B=Divide(A=@n13, B=@n14))) {name=PolarUV, nodeColor=<rgba 0 0.1094 1 1>}
  n15 = VectorParameter() {parameterName=Main_UV, defaultValue=<rgba 1 1 1 0>}
  NamedRerouteDeclaration(None=MaterialFunctionCall!(
    ClampTexture
    UV=NamedRerouteUsage()
    Tex
    InRGB=@n15.RGB
    InA=@n15.RGB
    Time=CollectionParameter() {collection=&MPC_PostProcress_Drain, parameterName=TextureTimeMAT}
    TimeMult
   ) {materialFunction=/Hermes/Props/ClayTile/ClayTile_DragonGate_B/VFX/MF_MOON_TexPanner_V2}.PannerUV) {name=MainUV, nodeColor=<rgba 0 0.4049 1 1>}
  n16 = VectorParameter() {parameterName=Tint_UV, defaultValue=<rgba 1 1 1 0>}
  NamedRerouteDeclaration(None=MaterialFunctionCall(
    UV=NamedRerouteUsage()
    InRGB=@n16.RGB
    InA=@n16.RGB
    Time=CollectionParameter() {collection=&MPC_PostProcress_Drain, parameterName=TextureTimeMAT}
   ) {materialFunction=&MF_MOON_TexPanner_V2}.PannerUV) {name=TintUV, nodeColor=<rgba 0 0.2806 1 1>}
  n17 = TextureCoordinate()
  n18 = TextureCoordinate() {coordinateIndex=1}
  n19 = ScalarParameter() {parameterName=CubeMap_Switch}
  n20 = Subtract(A=Multiply(A=LinearInterpolate(A=@n17, B=@n18, Alpha=@n19)) {constB=2})
  n21 = Multiply(A=@n20, B=@n20) {constB=2}
  n22 = MaterialFunctionCall(Float2=OneMinus(None=Abs(None=@n21))) {materialFunction=&BreakOutFloat2Components}
  NamedRerouteDeclaration(None=MaterialFunctionCall(A=@n22.R, B=Multiply(A=@n22.R, B=@n22.R), C=@n22.R, Alpha=ScalarParameter() {defaultValue=0.5, parameterName=CubeMap_Select}) {materialFunction=&Lerp_3Color}.Result) {name=CubeMap_2, nodeColor=<rgba 1 1 1 1>}
  n23 = ScalarParameter() {defaultValue=8, parameterName=Param_4}
  n24 = MaterialFunctionCall!(
   Sphere Radius=@n23
   Sphere Falloff
   Unit Ray Direction
   Sphere Center=CollectionParameter() {collection=&MPC_PostProcress_Drain, parameterName=Player_WP}
   Ray Origin
  ) {materialFunction=/Engine/Functions/Engine_MaterialFunctions02/Gradients/RayTraceSphereFalloff}
  n25 = Saturate(None=OneMinus(None=Divide(A=@n24.Unclamped Results, B=@n23)))
  NamedRerouteDeclaration(None=@n25) {name=TestColor, nodeColor=<rgba 0.5938 0 1 1>}
  NamedRerouteDeclaration(None=MaterialFunctionCall!(Camera Relative Position=@n10, X Camera to World Vector, Y Camera to World Vector, Z Camera to World Vector, "Tan(FOV/2)*[1,Screen Res Y/X]") {materialFunction=/Engine/Functions/Engine_MaterialFunctions02/Math/ManualWorldToScreenUVsTransform}.Screen UVs) {name=EffectUVPos, nodeColor=<rgba 0 0.3425 1 1>}
  n26 = CollectionParameter() {collection=&MPC_PostProcress_Drain, parameterName=GateTexture}
  n27 = TextureSample!(UVs=NamedRerouteUsage(), Tex, Level, Apply View MipBias) {mipValueMode=TMVM_MipLevel, samplerType=SAMPLERTYPE_LinearColor, texture=/Hermes/Props/ClayTile/ClayTile_DragonGate_B/VFX/Tile_output_8}
  n28 = Multiply(
   A=SphereMask(
    A=WorldPosition().XYZ
    B=VectorParameter() {parameterName=Gate_WP, defaultValue=<rgba -57600 42904 9988 0>}.RGB
    Radius=ScalarParameter() {defaultValue=15000, parameterName=Param_2}
   ) {hardnessPercent=0}
   B=Multiply(A=@n27.G, B=@n27.G)
  )
  n29 = ComponentMask(None=@n26) {a=true}
  n30 = NamedRerouteUsage()
  n31 = ScalarParameter() {parameterName=Param_7}
  n32 = TextureSample(UVs=Add(A=NamedRerouteUsage(), B=Multiply(A=@n30, B=@n31))) {mipValueMode=TMVM_MipLevel, samplerType=SAMPLERTYPE_LinearColor, texture=&Tile_output_8}
  n33 = Add(A=@n32.G, B=Add(A=NamedRerouteUsage(), B=ScalarParameter() {parameterName=Param_6}))
  n34 = CollectionParameter() {collection=&MPC_PostProcress_Drain, parameterName=OverlayColor}
  n35 = SceneTexture() {sceneTextureId=PPI_PostProcessInput0}
  n36 = CollectionParameter() {collection=&MPC_PostProcress_Drain, parameterName=BlurIntensity_B}
  n37 = CollectionParameter() {collection=&MPC_PostProcress_Drain, parameterName=BlurIntensity_A}
  n38 = TextureSample(UVs=NamedRerouteUsage()) {
   mipValueMode=TMVM_MipLevel
   samplerType=SAMPLERTYPE_LinearColor
   texture=/Game/Effects/Fort_Effects/FrontEnd/S33/GateContent/T_FX_Tile_0138_Moon
  }
  n39 = Saturate(None=Add(A=@n2))
  n40 = ScalarParameter() {defaultValue=0.25, parameterName=Param_5}
  n41 = NamedRerouteDeclaration(None=Multiply(A=@n38.G, B=Power(Base=@n39, Exp=@n40))) {name=DistortIntensity, nodeColor=<rgba 1 1 1 1>}
  n42 = ScreenPosition()
  n43 = AppendVector(A=NamedRerouteUsage(), B=Constant() {r=1})
  n44 = MaterialFunctionCall!(
   Texture Scale=LinearInterpolate(B=LinearInterpolate(A=@n36, B=@n37, Alpha=@n41) {constA=1}, Alpha=NamedRerouteUsage()) {constA=1}
   UVs=@n42.ViewportUV
   Offset XY Texturesize Z=@n43
  ) {materialFunction=/Engine/Functions/Engine_MaterialFunctions01/Texturing/ScaleUVsAroundPoint}
  n45 = SceneTexture(UVs=LinearInterpolate(A=ScreenPosition().ViewportUV, B=ComponentMask(None=@n44.UVs) {g=true, r=true}, Alpha=NamedRerouteUsage())) {sceneTextureId=PPI_PostProcessInput0}
  n46 = CollectionParameter() {collection=&MPC_PostProcress_Drain, parameterName=BlurIntensity_B}
  n47 = CollectionParameter() {collection=&MPC_PostProcress_Drain, parameterName=BlurIntensity_A}
  n48 = Multiply(
   A=LinearInterpolate(Alpha=MaterialFunctionCall() {materialFunction=&DitherTemporalAA}.Result) {constA=0.75}
   B=@n41
  )
  n49 = MaterialFunctionCall(
   Texture Scale=LinearInterpolate(B=LinearInterpolate(A=@n46, B=@n47, Alpha=@n48) {constA=1}, Alpha=NamedRerouteUsage()) {constA=1}
   UVs=@n42.ViewportUV
   Offset XY Texturesize Z=@n43
  ) {materialFunction=&ScaleUVsAroundPoint}
  n50 = SceneTexture(UVs=LinearInterpolate(A=ScreenPosition().ViewportUV, B=ComponentMask(None=@n49.UVs) {g=true, r=true}, Alpha=NamedRerouteUsage())) {sceneTextureId=PPI_PostProcessInput0}
  n51 = Multiply(A=NamedRerouteUsage()) {constB=-1}
  n52 = Multiply(
   A=Multiply(
    A=LinearInterpolate(A=@n45.Color, B=@n50.Color)
    B=CollectionParameter() {collection=&MPC_PostProcress_Drain, parameterName=GhostColor}
   )
   B=Multiply(A=Saturate(None=@n51), B=ScalarParameter() {defaultValue=1})
  )
  n53 = Add(
   A=Multiply(A=Saturate(None=@n33), B=ComponentMask(None=@n34) {b=true, g=true, r=true})
   B=ComponentMask(None=Add(A=@n35.Color, B=@n52)) {b=true, g=true, r=true}
  )
  n54 = Saturate(None=ComponentMask(None=SceneTexture() {sceneTextureId=PPI_PostProcessInput0}.Color) {b=true, g=true, r=true})
  n55 = VectorParameter() {parameterName=ColorPicker, defaultValue=<rgba 0 0.4774 1 1>}
  n56 = LinearInterpolate(A=@n53, B=Desaturation(None=@n53) {luminanceFactors=<rgba 0.2126 0.7152 0.0722 1>}, Alpha=OneMinus(None=DotProduct(A=@n54, B=@n55.RGB)))
  n57 = ComponentMask(None=@n1) {a=true}
  NamedRerouteDeclaration(None=LinearInterpolate(A=LinearInterpolate(A=@n53, B=@n56, Alpha=@n57), B=ComponentMask(None=@n26) {b=true, g=true, r=true}, Alpha=Multiply(A=@n28, B=@n29))) {name=EffectImage, nodeColor=<rgba 0 0.6953 1 1>}
  n58 = WorldPosition()
  n59 = VectorParameter() {parameterName=WP, defaultValue=<rgba -57600 42904 9988 0>}
  n60 = MaterialFunctionCall(Float2=Subtract(A=@n58.XYZ, B=@n59.RGB)) {materialFunction=&BreakOutFloat2Components}
  n61 = AppendVector(
   A=Arctangent2Fast(Y=@n60.R, X=@n60.R)
   B=SphereMask(A=@n58.XYZ, B=@n59.RGB, Radius=ScalarParameter() {defaultValue=40000, parameterName=Param_9}) {attenuationRadius=100000, hardnessPercent=0}
  )
  n62 = VectorParameter() {parameterName=Ground_UV, defaultValue=<rgba 1 1 1 0>}
  NamedRerouteDeclaration(None=MaterialFunctionCall(
    UV=NamedRerouteDeclaration(None=@n61) {name=GroundUV, nodeColor=<rgba 0.0147 0 1 1>}
    InRGB=@n62.RGB
    InA=@n62.RGB
    Time=CollectionParameter() {collection=&MPC_PostProcress_Drain, parameterName=TextureTimeMAT}
   ) {materialFunction=&MF_MOON_TexPanner_V2}.PannerUV) {name=GroundPannerUV, nodeColor=<rgba 0 0.2806 1 1>}
  NamedRerouteDeclaration(None=@n25) {name=PlayerMask, nodeColor=<rgba 0.5234 1 0 1>}
  ScalarParameter() {defaultValue=0.1, parameterName=Param_1}
  VectorParameter() {parameterName=Player_WP, defaultValue=<rgba -67953.8203 31213.8535 12664.1758 0>}
  VectorParameter() {parameterName=Param_2, defaultValue=<rgba 0.2553 0.08 0.678 1>}
 ]
}
