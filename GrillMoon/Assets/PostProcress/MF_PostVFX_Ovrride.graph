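# Material function that lets a global switch override the emissive color and opacity of any material's attributes, apparently for post-process VFX.
# It blends the incoming emissive toward the EmissiveOvrride RGB and opacity toward fully opaque, using PostSwitch as the alpha.
# Both parameters come from the MPC_PostProcress collection, so one change affects every material using the function.
# The Emi vector parameter is left unconnected.
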MaterialFunction /GrillMoon/Assets/PostProcress/MF_PostVFX_Ovrride
graph {nodes [
  n1 = GetMaterialAttributes(None=FunctionInput() {bUsePreviewValueAsDefault=true, inputName=InMA, inputType=FunctionInput_MaterialAttributes}) {attributeGetTypes [#1, #2]}
  n2 = NamedRerouteUsage()
  n3 = CollectionParameter() {collection=&/MPC_PostProcress, parameterName=PostSwitch}
  n4 = NamedRerouteUsage()
  FunctionOutput(None=SetMaterialAttributes!(
    MaterialAttributes=@n1.MaterialAttributes
    EmissiveColor=LinearInterpolate(A=@n1.EmissiveColor, B=@n2, Alpha=@n3)
    Opacity=LinearInterpolate(A=@n1.Opacity, B=@n4, Alpha=@n3)
   ) {attributeSetTypes [#1, #2]}) {outputName=OutMA}
  NamedRerouteDeclaration(None=ComponentMask(None=CollectionParameter() {collection=&MPC_PostProcress, parameterName=EmissiveOvrride}) {b=true, g=true, r=true}) {name=Emissive, nodeColor=<rgba 1 1 1 1>}
  NamedRerouteDeclaration(None=Constant() {r=1}) {name=Opacity, nodeColor=<rgba 1 1 1 1>}
  VectorParameter() {parameterName=Emi, defaultValue=<rgba 1 1 0 1>}
 ]}
