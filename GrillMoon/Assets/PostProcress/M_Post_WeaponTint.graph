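# M_Post_WeaponTint is a lit surface material that, by its name, serves a weapon tint effect.
# A scalar parameter drives CameraOffset, whose clamped output becomes World Position Offset to shift the mesh relative to the camera.
# Those attributes pass through the MF_PostVFX_Ovrride function, which holds any further look logic.
# The function's emissive output is then extracted, multiplied by EmissiveScale and written back into the material attributes.
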Material /GrillMoon/Assets/PostProcress/M_Post_WeaponTint
graph {
 properties {bAutomaticallySetUsageInEditor=false, bUseMaterialAttributes=true, usedShadingModels=Default Lit}
 nodes [
  n1 = ScalarParameter() {defaultValue=100}
  n2 = MaterialFunctionCall!(InMA=SetMaterialAttributes!(
    MaterialAttributes
    World Position Offset=MaterialFunctionCall(Offset Amount=@n1) {materialFunction=/Engine/Functions/Engine_MaterialFunctions02/WorldPositionOffset/CameraOffset}.Clamped Camera Offset
   ) {attributeSetTypes [#1]}) {materialFunction=&/MF_PostVFX_Ovrride}
  SetMaterialAttributes!(
   MaterialAttributes=@n2.OutMA
   Emissive Color=Multiply(
    A=GetMaterialAttributes(None=@n2.OutMA) {attributeGetTypes [#2]}.Emissive Color
    B=ScalarParameter() {defaultValue=1, parameterName=EmissiveScale}
   )
  ) {attributeSetTypes [#2]}
 ]
}
